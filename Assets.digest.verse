# Copyright Epic Games, Inc. All Rights Reserved.
#################################################
# Generated Digest of Verse API
# DO NOT modify this manually!
# Generated from build: ++Fortnite+Release-36.20-CL-43847582
#################################################


using {/Verse.org/Colors}
using {/Verse.org/Simulation}
using {/Verse.org/Assets}
using {/Verse.org/Native}
using {/Verse.org/SceneGraph}
# Module import path: /playkigai.dig@fortnite.com/Minigame_Chess/EPs
EPs := module:
    @import_as("/Minigame_Chess/EPs/chess_board.chess_board_C")
    chess_board<scoped {/playkigai.dig@fortnite.com/Minigame_Chess}> := class<final>(entity):


    chess_board_asset<scoped {/playkigai.dig@fortnite.com/Minigame_Chess}>:entity_prefab = external {}

    @import_as("/Minigame_Chess/EPs/chess_board_tile.chess_board_tile_C")
    chess_board_tile<scoped {/playkigai.dig@fortnite.com/Minigame_Chess}> := class<final>(entity):


    chess_board_tile_asset<scoped {/playkigai.dig@fortnite.com/Minigame_Chess}>:entity_prefab = external {}

    @import_as("/Minigame_Chess/EPs/chess_game_template.chess_game_template_C")
    chess_game_template<scoped {/playkigai.dig@fortnite.com/Minigame_Chess}> := class<final>(entity):


    chess_game_template_asset<scoped {/playkigai.dig@fortnite.com/Minigame_Chess}>:entity_prefab = external {}

    @import_as("/Minigame_Chess/EPs/chess_piece_bishop.chess_piece_bishop_C")
    chess_piece_bishop<scoped {/playkigai.dig@fortnite.com/Minigame_Chess}> := class<final>(entity):


    chess_piece_bishop_asset<scoped {/playkigai.dig@fortnite.com/Minigame_Chess}>:entity_prefab = external {}

    @import_as("/Minigame_Chess/EPs/chess_piece_king.chess_piece_king_C")
    chess_piece_king<scoped {/playkigai.dig@fortnite.com/Minigame_Chess}> := class<final>(entity):


    chess_piece_king_asset<scoped {/playkigai.dig@fortnite.com/Minigame_Chess}>:entity_prefab = external {}

    @import_as("/Minigame_Chess/EPs/chess_piece_knight.chess_piece_knight_C")
    chess_piece_knight<scoped {/playkigai.dig@fortnite.com/Minigame_Chess}> := class<final>(entity):


    chess_piece_knight_asset<scoped {/playkigai.dig@fortnite.com/Minigame_Chess}>:entity_prefab = external {}

    @import_as("/Minigame_Chess/EPs/chess_piece_pawn.chess_piece_pawn_C")
    chess_piece_pawn<scoped {/playkigai.dig@fortnite.com/Minigame_Chess}> := class<final>(entity):


    chess_piece_pawn_asset<scoped {/playkigai.dig@fortnite.com/Minigame_Chess}>:entity_prefab = external {}

    @import_as("/Minigame_Chess/EPs/chess_piece_queen.chess_piece_queen_C")
    chess_piece_queen<scoped {/playkigai.dig@fortnite.com/Minigame_Chess}> := class<final>(entity):


    chess_piece_queen_asset<scoped {/playkigai.dig@fortnite.com/Minigame_Chess}>:entity_prefab = external {}

    @import_as("/Minigame_Chess/EPs/chess_piece_rook.chess_piece_rook_C")
    chess_piece_rook<scoped {/playkigai.dig@fortnite.com/Minigame_Chess}> := class<final>(entity):


    chess_piece_rook_asset<scoped {/playkigai.dig@fortnite.com/Minigame_Chess}>:entity_prefab = external {}

    @import_as("/Minigame_Chess/EPs/chess_time_clock.chess_time_clock_C")
    chess_time_clock<scoped {/playkigai.dig@fortnite.com/Minigame_Chess}> := class<final>(entity):


    chess_time_clock_asset<scoped {/playkigai.dig@fortnite.com/Minigame_Chess}>:entity_prefab = external {}

# Module import path: /playkigai.dig@fortnite.com/Minigame_Chess/Materials
Materials := module:
    M_ChessPiece<scoped {/playkigai.dig@fortnite.com/Minigame_Chess}> := class<final><scoped {/playkigai.dig@fortnite.com/Minigame_Chess}>(material):
        @editable
        var Roughness<public>:float = external {}

        @editable
        var Color<public>:color = external {}

    M_Clock_Body<scoped {/playkigai.dig@fortnite.com/Minigame_Chess}> := class<final><scoped {/playkigai.dig@fortnite.com/Minigame_Chess}>(material):
        @editable
        var Param<public>:color = external {}

    M_Clock_Screen<scoped {/playkigai.dig@fortnite.com/Minigame_Chess}> := class<final><scoped {/playkigai.dig@fortnite.com/Minigame_Chess}>(material):
        @editable
        var Param<public>:color = external {}

    MI_ChessPiece_Black<scoped {/playkigai.dig@fortnite.com/Minigame_Chess}> := class<final><scoped {/playkigai.dig@fortnite.com/Minigame_Chess}>(material):
        @editable
        var Roughness<public>:float = external {}

        @editable
        var Color<public>:color = external {}

    MI_ChessPiece_BlackGlow<scoped {/playkigai.dig@fortnite.com/Minigame_Chess}> := class<final><scoped {/playkigai.dig@fortnite.com/Minigame_Chess}>(material):
        @editable
        var Roughness<public>:float = external {}

        @editable
        var Color<public>:color = external {}

    MI_ChessPiece_White<scoped {/playkigai.dig@fortnite.com/Minigame_Chess}> := class<final><scoped {/playkigai.dig@fortnite.com/Minigame_Chess}>(material):
        @editable
        var Roughness<public>:float = external {}

        @editable
        var Color<public>:color = external {}

    MI_ChessPiece_WhiteGlow<scoped {/playkigai.dig@fortnite.com/Minigame_Chess}> := class<final><scoped {/playkigai.dig@fortnite.com/Minigame_Chess}>(material):
        @editable
        var Roughness<public>:float = external {}

        @editable
        var Color<public>:color = external {}

    MI_PauseButton<scoped {/playkigai.dig@fortnite.com/Minigame_Chess}> := class<final><scoped {/playkigai.dig@fortnite.com/Minigame_Chess}>(material):
        @editable
        var Roughness<public>:float = external {}

        @editable
        var Color<public>:color = external {}

    MI_RefreshButton<scoped {/playkigai.dig@fortnite.com/Minigame_Chess}> := class<final><scoped {/playkigai.dig@fortnite.com/Minigame_Chess}>(material):
        @editable
        var Roughness<public>:float = external {}

        @editable
        var Color<public>:color = external {}

# Module import path: /playkigai.dig@fortnite.com/Minigame_Chess/Meshes
Meshes := module:
    SM_Chess_Bishop<scoped {/playkigai.dig@fortnite.com/Minigame_Chess}> := class<final><scoped {/playkigai.dig@fortnite.com/Minigame_Chess}>(mesh_component):
        @editable
        var Material<public>:material = external {}

    SM_Chess_Bishop_asset<scoped {/playkigai.dig@fortnite.com/Minigame_Chess}>:mesh = external {}

    SM_Chess_King<scoped {/playkigai.dig@fortnite.com/Minigame_Chess}> := class<final><scoped {/playkigai.dig@fortnite.com/Minigame_Chess}>(mesh_component):
        @editable
        var Material<public>:material = external {}

    SM_Chess_King_asset<scoped {/playkigai.dig@fortnite.com/Minigame_Chess}>:mesh = external {}

    SM_Chess_Knight<scoped {/playkigai.dig@fortnite.com/Minigame_Chess}> := class<final><scoped {/playkigai.dig@fortnite.com/Minigame_Chess}>(mesh_component):
        @editable
        var Material<public>:material = external {}

    SM_Chess_Knight_asset<scoped {/playkigai.dig@fortnite.com/Minigame_Chess}>:mesh = external {}

    SM_Chess_Pawn<scoped {/playkigai.dig@fortnite.com/Minigame_Chess}> := class<final><scoped {/playkigai.dig@fortnite.com/Minigame_Chess}>(mesh_component):
        @editable
        var Material<public>:material = external {}

    SM_Chess_Pawn_asset<scoped {/playkigai.dig@fortnite.com/Minigame_Chess}>:mesh = external {}

    SM_Chess_Queen<scoped {/playkigai.dig@fortnite.com/Minigame_Chess}> := class<final><scoped {/playkigai.dig@fortnite.com/Minigame_Chess}>(mesh_component):
        @editable
        var Material<public>:material = external {}

    SM_Chess_Queen_asset<scoped {/playkigai.dig@fortnite.com/Minigame_Chess}>:mesh = external {}

    SM_Chess_Rook<scoped {/playkigai.dig@fortnite.com/Minigame_Chess}> := class<final><scoped {/playkigai.dig@fortnite.com/Minigame_Chess}>(mesh_component):
        @editable
        var Material<public>:material = external {}

    SM_Chess_Rook_asset<scoped {/playkigai.dig@fortnite.com/Minigame_Chess}>:mesh = external {}

    SM_Chess_Tile<scoped {/playkigai.dig@fortnite.com/Minigame_Chess}> := class<final><scoped {/playkigai.dig@fortnite.com/Minigame_Chess}>(mesh_component):
        @editable
        var Material<public>:material = external {}

    SM_Chess_Tile_asset<scoped {/playkigai.dig@fortnite.com/Minigame_Chess}>:mesh = external {}

    SM_Clock_Body<scoped {/playkigai.dig@fortnite.com/Minigame_Chess}> := class<final><scoped {/playkigai.dig@fortnite.com/Minigame_Chess}>(mesh_component):
        @editable
        var Material<public>:material = external {}

        @editable
        var Material_001<public>:material = external {}

    SM_Clock_Body_asset<scoped {/playkigai.dig@fortnite.com/Minigame_Chess}>:mesh = external {}

    SM_Clock_Button<scoped {/playkigai.dig@fortnite.com/Minigame_Chess}> := class<final><scoped {/playkigai.dig@fortnite.com/Minigame_Chess}>(mesh_component):
        @editable
        var Material<public>:material = external {}

    SM_Clock_Button_asset<scoped {/playkigai.dig@fortnite.com/Minigame_Chess}>:mesh = external {}

# Module import path: /playkigai.dig@fortnite.com/Minigame_Chess/Sounds
Sounds := module:
    impactTin_medium_000_asset<public>:sound_wave = external {}

    impactTin_medium_001_asset<public>:sound_wave = external {}

    impactTin_medium_002_asset<public>:sound_wave = external {}

    impactTin_medium_003_asset<public>:sound_wave = external {}

    impactTin_medium_004_asset<public>:sound_wave = external {}

    PieceMove<scoped {/playkigai.dig@fortnite.com/Minigame_Chess}> := class<final><public>(sound_component):
        @editable
        var (/playkigai.dig@fortnite.com/Minigame_Chess/Sounds/PieceMove:)Sounds<public>:[]sound_wave = external {}

        @editable
        var PitchBase<public>:float = external {}

        @editable
        var PitchRandomSpread<public>:float = external {}

        @editable
        var VolumeBase<public>:float = external {}

        @editable
        var VolumeRandomSpread<public>:float = external {}