<#
@UUID: ead44cb4-3904-4080-b066-ec9bc545ce99
@Task: Chess Board Manager Component
@Description: Scene graph component that procedurally generates the chess board and piece layout and also handles the player interaction and movement of pieces on the board.
@Tags: Gameplay, SceneGraph, TeamManagement, DataStructures, ProceduralGeneration, PlayerInteraction, Movement
@Export: true
#>

<#
In-Editor Entity Prefab Setup:
    (NOTE: any setting not mentioned is assumed to be default or to your preference)

chess_piece_[PIECE NAME]
    RootEntity
        - ANY COMPONENT THAT INHERITS FROM "chess_piece_component"
        - mesh_component (using custom SM_Chess_[PIECE NAME] static mesh)
            - Material: M_ChessPiece_White (will change programatically when board is generated)
            - Collidable: false (must be disabled to allow for interactable collider)
            - Queryable: false
        - transform_component
        PieceInteractCollider
            - interactable_component
            - mesh_component (preferably a cylinder)
            - transform_component
                - Visible: false
                - Tweak the scale values so that it is just slightly large enough to envelop RootEntity's mesh component

    (NOTE: this entity prefab definition applies to all the pieces, just replace [PIECE NAME] with whatever the piece is.)

chess_board_tile
    RootEntity
        - mesh_component (using custom SM_Chess_Tile static mesh)
            - Material: M_ChessPiece_Black (will change programatically when board is generated)
            - Collidable: false (must be disabled to allow for interactable collider)
            - Queryable: false
        - transform_component
        TileInteractCollider
            - interactable_component
            - mesh_component (preferably a cube)
            - transform_component
                - Visible: false
                - Tweak the scale values so that it is just slightly large enough to envelop RootEntity's mesh component

chess_board
    RootEntity
        - chess_board_component
        - PieceMove (sound_component)
            - Sounds: populate array with all the chess piece move sounds you want/need
        - transform_component
#>

using { /Fortnite.com/Game }
using { /Fortnite.com/Devices }
using { /Verse.org/SceneGraph }
using { /Verse.org/SpatialMath }
using { /UnrealEngine.com/Temporary/SpatialMath }
using { /Verse.org/Simulation }
using { /Verse.org/Assets }
using {EPs}


(BoardEntity: chess_board).GetBoardComponent()<decides><transacts>:chess_board_component =
    BoardEntity.GetDescendantComponentsArray(chess_board_component)[0]


chess_board_component<public> := class<final_super>(component):
    
    @editable_number(float):
        MinValue := option. 0.0
    BoardScale : float = 2.0
    
    NumRows : int = 8
    NumColumns : int = 8

    # Signals when a player has selected where a piece will moves and when it begins moving across the board
    PieceMoveStartEvent<public>: event() = event(){}
    # Signals when a piece completes its move 
    PieceMoveEndEvent<public>: event() = event(){}
    # Signals when a player selects a piece to be moved across the board; passes component of the piece that was selected
    PieceSelectedEvent<public>: event(chess_piece_component) = event(chess_piece_component){}
    # Signals when the a piece is unselected; passes component of the piece that was selected
    PieceDeselectedEvent<public>: event(chess_piece_component) = event(chess_piece_component){}
    # Signals when a side's king is captured; passes the side whose king was captured
    KingCaptured<public>: event(chess_side) = event(chess_side){}
    # Signals when this component's OnEndSimulation function fires; used to cancel asyncronous logic when this component stops simulating
    EndSimulationEvent<public>: event() = event(){}
    
    var MaybeGameManager: ?chess_game_manager_component = false
    var ChessBoard: [][]chess_tile_component = array{}
    var Pieces : []chess_piece_component = array{}
    var TargetablePositions: []chess_tile_component = array{}
    var MaybeSelectedPiece : ?chess_piece_component = false
    var isInFreeplayMode : logic = true

    
    OnBeginSimulation<override>():void =
        (super:)OnBeginSimulation()
        set MaybeGameManager = option. Entity.GetAncestorComponentsArray(chess_game_manager_component)[0]

        if (RoundMan := Entity.GetFortRoundManager[]):
            RoundMan.SubscribeRoundStarted(OnRoundStart)


    OnEndSimulation<override>():void =
        (super:)OnEndSimulation()
        EndSimulationEvent.Signal()


    OnRoundStart()<suspends>:void=
        race:
            EndSimulationEvent.Await()
            block:
                RefreshBoard()
                # Listen for Game Manager Events
                if (GameManager := MaybeGameManager?):
                    race:
                        loop:   # On Game Start
                            GameManager.GameBeginEvent.Await()
                            RefreshBoard()
                            DeselectPiece()
                            # Disable freeplay and the refresh board button
                            set isInFreeplayMode = false
                            Sleep(0.0)
                        loop:   # On Turn Start
                            CurrentTurn := GameManager.TurnStartEvent.Await()
                            DeselectPiece()
                            Sleep(0.0)
                        loop:   # On Game Over
                            MaybeWinner := GameManager.GameOverEvent.Await()
                            DeselectPiece()
                            Sleep(0.0)
                        loop:   # On Post Game Delay
                            GameManager.PostGameDelayCompleteEvent.Await()
                            # Enable freeplay again and the refresh board button
                            set isInFreeplayMode = true
                            Sleep(0.0)
                    
        
    GetTile(Row : int, Column : int)<decides><transacts> : chess_tile_component=
        ChessBoard[Row][Column]


    PieceInteractedWith(Piece : chess_piece_component)<suspends>:void=
        # Interacting with a targeted piece we want to capture.
        if: 
            Piece.Targeted?
            SelectedPiece := MaybeSelectedPiece?
            TargetTile := Piece.MaybeTile?
        then:
            MovePiece(SelectedPiece, TargetTile)
        
        # Interacting with the piece we already have selected
        else if(Piece.Selected?): 
            DeselectPiece()
        
        # Interacting with a different piece than the one already selected
        else: 
            DeselectPiece()
            SelectPiece(Piece) 


    SelectPiece(NewPiece : chess_piece_component)<suspends>:void=
        # Only select the piece if it can move to at least one tile
        set TargetablePositions = NewPiece.GetTargetablePositions()
        if(TargetablePositions.Length > 0):
            NewPiece.SelectPiece()
            for(Tile : TargetablePositions):
                Tile.Highlight()
        else:
            Print("chess_board_component:\nCan't Move!")
       
        set MaybeSelectedPiece = option. NewPiece

        
    DeselectPiece()<suspends>:void=
        if (Selected := MaybeSelectedPiece?):
            Selected.DeselectPiece()
            for(Tile : TargetablePositions):
                Tile.Unhighlight()

            set MaybeSelectedPiece = false
            set TargetablePositions = array{}        


    MovePiece(Piece : chess_piece_component, Tile : chess_tile_component)<suspends>:void=
        race:
            EndSimulationEvent.Await()
            block: 
                PieceMoveStartEvent.Signal()
                PieceTransform := Piece.Entity.GetLocalTransform()
                StartPos := PieceTransform.Translation
                var MaybeEndPos : ?(/Verse.org/SpatialMath:)vector3 = false
                set MaybeEndPos = option. (/Verse.org/SpatialMath:) vector3 {
                    Forward := Tile.Entity.GetLocalTransform().Translation.Forward, 
                    Left := Tile.Entity.GetLocalTransform().Translation.Left, 
                    Up := Tile.Entity.GetLocalTransform().Translation.Up + 25.0
                    }

                # Movement Animation
                Timestamp := GetSimulationElapsedTime()
                loop:
                    Sleep(0.0)
                    DeltaTime := GetSimulationElapsedTime() - Timestamp
                    Interp := Clamp(Piece.MoveSpeed*DeltaTime, 0.0, 1.0)

                    if(EndPos := MaybeEndPos?):
                        NewTransform :=  (/Verse.org/SpatialMath:)transform {
                            Translation := Lerp(StartPos, EndPos, Interp),
                            Rotation := PieceTransform.Rotation,
                            Scale := PieceTransform.Scale
                        }
                        Piece.Entity.SetLocalTransform(NewTransform)
                    if(Interp = 1.0):
                        break

                # Capture any opposing piece at the target location
                RemovePieceAtLocation(Tile)

                set Tile.MaybeChessPieceOnTile = option. Piece

                if(PreviousTile := Piece.MaybeTile?):
                    set PreviousTile.MaybeChessPieceOnTile = false

                set Piece.MaybeTile = option. Tile
                PieceMoveEndEvent.Signal()

                # Play a move sound
                SoundComp := Entity.FindDescendantComponents(sound_component)
                for(ThisSoundComp : SoundComp):
                    ThisSoundComp.Play()

                DeselectPiece()

                # Promote pawns to queens if they reach the opposite side of the board
                if(Pawn := chess_pawn_component[Piece]):
                    set Pawn.MovedOnce = true
                    case (Piece.Side):
                        chess_side.White => if (Tile.Row = 7) then Pawn.Upgrade()
                        chess_side.Black => if (Tile.Row = 0) then Pawn.Upgrade()
   

    RemovePieceAtLocation(Tile : chess_tile_component):void=
        if(ChessPieceComp := Tile.MaybeChessPieceOnTile?):
            # If the piece removed is a king, end the game
            if:
                chess_king_component[ChessPieceComp]
                GameManager := MaybeGameManager?
            then:
                KingCaptured.Signal(ChessPieceComp.Side)

            # Remove the piece if there's a piece on that slot
            ChessPieceComp.Entity.RemoveFromParent()


    RefreshBoard():void =
        if(ChessBoard.Length = 0):
            set ChessBoard = MakeTiles()
        
        # Reset Tiles as having no piece.
        for (RowIndex -> Row : ChessBoard):
            # Inner loop iterates through columns in the current row
            for(ColumnIndex -> Tile : Row):
                set Tile.MaybeChessPieceOnTile = false

        # Destroy Old Pieces
        for(Piece : Pieces) do Piece.Entity.RemoveFromParent()

        # Create New Pieces
        set Pieces = MakePieces()
        for(Piece : Pieces) do AddPieceToBoard(Piece)
    

    MakeTile(Row : int, Column : int, Color : chess_side) : entity= 
        Tile := chess_board_tile{}
        if(TileComp := Tile.GetComponent[chess_tile_component]):
            TileComp.InitTile(Self, Row, Column, Color)
        Tile


    MakeTiles():[][]chess_tile_component =
        var TileComponents : [][]chess_tile_component = array{}
        
        # Make Tile Entities
        # Outer loop iterates through rows
        for (RowIndex := 0..NumRows-1):
            var RowStartColor : chess_side = chess_side.White
                if(Mod[RowIndex, 2] = 1):
                    set RowStartColor = chess_side.Black

            var TileCompRow : [] chess_tile_component = array{}

            # Inner loop iterates through columns in the current row
            for (ColumnIndex := 0..NumColumns-1):
                var TileColor : chess_side = RowStartColor
                if(Mod[ColumnIndex, 2] = 1):
                    set TileColor = TileColor.Opponent()

                Tile := MakeTile(RowIndex, ColumnIndex, TileColor)

                Entity.AddEntities(array{Tile})
                
                if (InteractableComponent := Tile.GetDescendantComponentsArray(interactable_component)[0]):
                    InteractableComponent.Disable()
                
                BoardShiftToCenterClock := (NumRows*100.0*BoardScale)/2.0
                BoardOffset := -BoardShiftToCenterClock+(100.0*BoardScale/2.0)

                NewTransform := (/Verse.org/SpatialMath:) transform {
                    Translation := (/Verse.org/SpatialMath:) vector3 {
                        Forward := RowIndex*100.0*BoardScale+BoardOffset, 
                        Left := ColumnIndex*100.0*BoardScale, 
                        Up := 0.0
                        },
                        Rotation := (/Verse.org/SpatialMath:) MakeRotationFromYawPitchRollDegrees(0.0, 0.0, 0.0),
                        Scale := (/Verse.org/SpatialMath:) vector3{Forward := BoardScale, Left := BoardScale, Up := BoardScale}
                    }

                Tile.SetLocalTransform(NewTransform)

                if(TileComp := Tile.GetComponent[chess_tile_component]):
                    set TileCompRow += array{TileComp}

            set TileComponents += array{TileCompRow}

        TileComponents


    MakePawn(Row : int, Column : int, Side : chess_side) : chess_piece_pawn=
        Piece := chess_piece_pawn{}
            if(PieceComp := Piece.GetComponent[chess_pawn_component]):
                PieceComp.InitPiece(Self, Row, Column, Side)
        Piece
    

    MakeRook(Row : int, Column : int, Side : chess_side) : chess_piece_rook=
        Piece := chess_piece_rook{}
            if(PieceComp := Piece.GetComponent[chess_rook_component]):
                PieceComp.InitPiece(Self, Row, Column, Side)
        Piece


    MakeKnight(Row : int, Column : int, Side : chess_side) : chess_piece_knight=
        Piece := chess_piece_knight{}
            if(PieceComp := Piece.GetComponent[chess_knight_component]):
                PieceComp.InitPiece(Self, Row, Column, Side)
        Piece


    MakeBishop(Row : int, Column : int, Side : chess_side) : chess_piece_bishop=
        Piece := chess_piece_bishop{}
            if(PieceComp := Piece.GetComponent[chess_bishop_component]):
                PieceComp.InitPiece(Self, Row, Column, Side)
        Piece


    MakeQueen(Row : int, Column : int, Side : chess_side) : chess_piece_queen=
        Piece := chess_piece_queen{}
            if(PieceComp := Piece.GetComponent[chess_queen_component]):
                PieceComp.InitPiece(Self, Row, Column, Side)
        Piece


    MakeKing(Row : int, Column : int, Side : chess_side) : chess_piece_king=
        Piece := chess_piece_king{}
            if(PieceComp := Piece.GetComponent[chess_king_component]):
                PieceComp.InitPiece(Self, Row, Column, Side)
        Piece


    MakePieces():[]chess_piece_component =
        var NewPieces : []entity = array{}

        # Create white backline entities
        set NewPieces += array{MakeRook(    0, 0, chess_side.White)}
        set NewPieces += array{MakeKnight(  0, 1, chess_side.White)}
        set NewPieces += array{MakeBishop(  0, 2, chess_side.White)}
        set NewPieces += array{MakeKing(    0, 3, chess_side.White)}
        set NewPieces += array{MakeQueen(   0, 4, chess_side.White)}
        set NewPieces += array{MakeBishop(  0, 5, chess_side.White)}
        set NewPieces += array{MakeKnight(  0, 6, chess_side.White)}
        set NewPieces += array{MakeRook(    0, 7, chess_side.White)}

        # Create white pawn frontline entities
        for (Column := 0..NumColumns-1):
            set NewPieces += array{MakePawn(1, Column, chess_side.White)}

        # Create black backline entities
        set NewPieces += array{MakeRook(    7, 0, chess_side.Black)}
        set NewPieces += array{MakeKnight(  7, 1, chess_side.Black)}
        set NewPieces += array{MakeBishop(  7, 2, chess_side.Black)}
        set NewPieces += array{MakeKing(    7, 3, chess_side.Black)}
        set NewPieces += array{MakeQueen(   7, 4, chess_side.Black)}
        set NewPieces += array{MakeBishop(  7, 5, chess_side.Black)}
        set NewPieces += array{MakeKnight(  7, 6, chess_side.Black)}
        set NewPieces += array{MakeRook(    7, 7, chess_side.Black)}

        # Create black pawn frontline entities
        for (Column := 0..NumColumns-1):            
            set NewPieces += array{MakePawn(6, Column, chess_side.Black)}
        
        # Get Piece Components
        NewPieceComponents := for (Piece : NewPieces, PieceComp := Piece.GetComponent[chess_piece_component]):
            PieceComp
        NewPieceComponents
        

    AddPieceToBoard(Piece : chess_piece_component):void =
        Entity.AddEntities(array{Piece.Entity})

        if(Tile := Piece.MaybeTile?):
            var Yaw : float = case (Piece.Side):
                chess_side.White => 90.0
                chess_side.Black => 270.0

            TileEntity := Tile.Entity
            set Tile.MaybeChessPieceOnTile = option. Piece
            NewTransform := (/Verse.org/SpatialMath:) transform {
                Translation := (/Verse.org/SpatialMath:) vector3 {
                    Forward := TileEntity.GetLocalTransform().Translation.Forward, 
                    Left := TileEntity.GetLocalTransform().Translation.Left, 
                    Up := TileEntity.GetLocalTransform().Translation.Up + 25.0
                    },
                Rotation := (/Verse.org/SpatialMath:) MakeRotationFromYawPitchRollDegrees(Yaw, 0.0, 0.0),
                Scale := (/Verse.org/SpatialMath:) vector3{Forward := 1.0, Left := 1.0, Up := 1.0}
                }
            Piece.Entity.SetLocalTransform(NewTransform)