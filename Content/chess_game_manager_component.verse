<#
@UUID: ead44cb4-3904-4080-b066-ec9bc545ce99
@Task: Chess Game Manager Component
@Description: Component that handles turn structure, game settings turn timer management, and all non-board related player interaction for the chess minigame
@Tags: Gameplay, SceneGraph, PlayerInteraction, DataStructures, TimerManagement, Devices
@Export: true
#>

<#
In-Editor Entity Prefab Setup:
    (NOTE: any setting not mentioned is assumed to be default or to your preference)

chess_board
    RootEntity
        - chess_board_component
        - PieceMove (sound_component)
            - Sounds: populate array with all the chess piece move sounds you want/need
        - transform_component

chess_time_clock
        RootEntity
            - mesh_component (using the SM_Clock_Body static mesh)
                - Material: M_Clock_Body
                - Material_001: M_Clock_Screen
            - transofrm_component

            ModeSelectionButton:
                - interactable_component
                - mesh_component (using the SM_button static mesh)
                    - WorldGridMaterial: M_PauseButton
                - transform_component
                    - tweak transoform until the button is poking out in the center below the clock screen dispaly

            RefreshButton:
                - interactable_component
                - mesh_component (using the SM_button static mesh)
                    - WorldGridMaterial: M_RefreshButton
                - transform_component
                    - tweak transoform until the button is poking out on the outer edge of the clock face panel
            
            WhiteSideReadyButton:
                - interactable_component
                - mesh_component (using the SM_button static mesh)
                    - WorldGridMaterial: M_ChessPiece_White
                - transform_component
                    - tweak transoform until the button is poking out to the right of the ModeSelectionButton
            
            BlackSideReadyButton:
                - interactable_component
                - mesh_component (using the SM_button static mesh)
                    - WorldGridMaterial: M_ChessPiece_Black
                - transform_component
                    - tweak transoform until the button is poking out to the left of the ModeSelectionButton

chess_game_template
    RootEntity
        - chess_game_manager_component
            - MaybeChessBoardEntity:
                - MaybeModeSelectionButton: set to "ModeSelectionButton"
                - MaybeWhiteReadyButton: set to "WhiteSideReadyButton"
                - MaybeBlackReadyButton: set to "BlackSideReadyButton"
                - MaybeRefreshButton: set to "RefreshButton"
                - TimeClockDisplay: set to billboard_device in the level 
                - PlayerWhiteStatusDisplay: set to billboard_device in the level 
                - PlayerBlackStatusDisplay: set to billboard_device in the level 
        - chess_hud_component

        ChessBoard (chess_board entity prefab)
            - transform_component
                - position at origin of Root Entity

        TimeClock (chess_time_clock entity prefab)
            - transform_component
                - position facing the chess_board compoennt and back far enough that it would not intersect with board when it generates (might require tiral and error)

#>

using { /Verse.org }
using { /Verse.org/Concurrency }
using { /Verse.org/Simulation }
using { /Verse.org/Native }
using { /Verse.org/SceneGraph }
using { /Fortnite.com/Game }
using { /Fortnite.com/Teams }
using { /Fortnite.com/Playspaces }
using { /Fortnite.com/Devices }
using { /UnrealEngine.com/Temporary/SpatialMath }
using { /UnrealEngine.com/Temporary/Diagnostics }


###################### Start Chess Game Manager Component #############################################################  

chess_game_state := enum {PreGame, GameInProgress, PostGame}


chess_game_manager_component<public> := class<final_super>(component):

    # ============ EDITABLE FIELDS ==================================
    # The amount of time in seconds after the game that this component will wait before reseting in anticipation for another
    @editable_number(float):
        MinValue := option. 0.0
        Categories := array{GameCategory}
    PostGameDelay: float = 10.0

    # Time in seconds that each player will get on their turn timer if the "Quick" Game speed is selected before the game
    @editable_number(float):
        MinValue := option. 0.0
        Categories := array{GameCategory}
    QuickGameTime: float = 300.0
    
    # Time in seconds that each player will get on their turn timer if the "Long" Game speed is selected before the game
    @editable_number(float):
        MinValue := option. 0.0
        Categories := array{GameCategory}
    LongGameTime: float = 600.0

    # When set to "true", it bypasses all need for player interaction for that side. Setting both to true, has game autoplay through all game phases with no player interaction.
    @editable:
        Categories := array{DebugCategory}
    WhiteInteractionOverride: logic = false
    @editable:
        Categories := array{DebugCategory}
    BlackInteractionOverride: logic = false

    # If the Debug Override value is set for a player, amount of artifical delay to apply to their actions before moving on to next game phase
    @editable_number(float):
        MinValue := option. 0.0
        Categories := array{DebugCategory}
    DebugStepDelay: float = 3.0
    
    # The root entity of the chess_board prefab and chess_board_manager component
    @editable:
        Categories := array{PrefabSetupCategory}
    MaybeChessBoard : ?EPs.chess_board = false

    # Entity with an interactable_component attached that cycles the game speed mode when interacted with during the pre-game phase
    @editable:
        Categories := array{PrefabSetupCategory}
    MaybeModeSelectButton: ?entity = false

    # Entity with an attached interactable_component that registers the interacting player to WHITE side and indicates that that WHITE is ready to begin the game.
    @editable:
        Categories := array{PrefabSetupCategory}
    MaybeWhiteReadyButton: ?entity = false

    # Entity with an attached interactable_component that registers the interacting player to BLACK side and indicates that that BLACK is ready to begin the game.
    @editable:
        Categories := array{PrefabSetupCategory}
    MaybeBlackReadyButton: ?entity = false
    
    # Entity with an attached interactable_component that resets the board when in free play mode
    @editable:
        Categories := array{PrefabSetupCategory}
    MaybeRefreshButton: ?entity = false

    # Billboard device used to display the current Game speed mode during the pregame and the turn timer of the current player's turn
    @editable:
        Categories := array{PrefabSetupCategory}
    TimeClockDisplay: billboard_device = billboard_device{}

    # Used to display when the white player is registered or not
    @editable:
        Categories := array{PrefabSetupCategory}
    PlayerWhiteStatusDisplay: billboard_device = billboard_device{}

    # Used to display when the black player is registered or not
    @editable:
        Categories := array{PrefabSetupCategory}
    PlayerBlackStatusDisplay: billboard_device = billboard_device{}

    # ============ EVENTS =======================================

    # Signaled when a player registers to a board side by interacting the respective Ready-up button; passes the updated chess_player_info object for that side.
    PlayerReadyEvent<public>: event(chess_player_info) = event(chess_player_info){}
    # Signaled when a player unregisters to a board side by interacting the respective Ready-up button after pressing it once already; passes the updated chess_player_info object for that side.
    PlayerUnreadyEvent<public>: event(chess_player_info) = event(chess_player_info){}
    # Signaled when two players have indicated that they are ready to begin the game
    BothPlayersReadyEvent<public>: event() = event(){}
    # Signaled when a player presses the Game Speed selector button to cycle through the Game Speed Modes; passes the new game speed.
    GameSpeedChangeEvent<public>: event(chess_game_speed) = event(chess_game_speed){}
    # Signaled when the game starts and the first turn is about to begin.
    GameBeginEvent<public>: event() = event(){}
    # Signaled at the start of each side's turn; passes the chess_player_info object reference of the current side's turn
    TurnStartEvent<public>: event(chess_player_info) = event(chess_player_info){}
    # Signaled when either side's turn time clock expires and before the game offiically ends; passes the chess_player_info object reference whose timer expired
    PlayerTimeClockExpired<public>: event(chess_player_info) = event(chess_player_info){}
    # Signals when one of the two registered players prematurely leaves the game; passes the chess_player_info of the side that abandoned
    PlayerConcedeEvent<public>: event(chess_player_info) = event(chess_player_info){}
    # Signals at the end of the game (regardless of the result); passes an optional chess_player_info object reference of the winning side (if any exists)
    GameOverEvent<public>: event(?chess_player_info) = event(?chess_player_info){}
    # Signals when the game manager fully reset in anticipation for the next game
    PostGameDelayCompleteEvent<public>: event() = event(){}
    # Signals when this component's OnEndSimulation function fires; used to cancel asyncronous logic when this component stops simulating
    EndSimulationEvent<internal>: event() = event(){}
    # Signals when this component's OnRemovedFromScene function fires; used to cancel asyncronous logic when this component stops simulating
    RemovedFromSceneEvent<internal>: event() = event(){}
    
    # ============ INTERNAL VARS =======================================
    # Which side its turn it currently is
    var CurrentTurn: chess_side = chess_side.White
    # How many rounds have passed since the game has begun
    var CurrentRound: int = 0
    # The speed mode of the game that can be set during pregame
    var GameSpeed: chess_game_speed = chess_game_speed.Quick

    # Data structures which contain all the information related to each side (the player, the )
    var WhiteSideInfo: chess_player_info = chess_player_info{ChessSide := chess_side.White}
    var BlackSideInfo: chess_player_info = chess_player_info{ChessSide := chess_side.Black}

    # Indicates which phase of the game loop we are in, Pre-Game, Game in-progress, or Post-Game
    var GameState: chess_game_state = chess_game_state.PreGame
    

    # Helper function that succeeds if it is currently this player's turn
    IsPlayersTurn<public>(Player : player)<decides><transacts>:void=
        Player = CurrentTurn.GetInfo().MaybePlayer?

        
    OnBeginSimulation<override>():void=
        (super:)OnBeginSimulation()

        if (RoundMan := Entity.GetFortRoundManager[]):
            RoundMan.SubscribeRoundStarted(OnRoundStart)

        if:
            WhiteButton := MaybeWhiteReadyButton?
            WhiteButtonMesh := WhiteButton.GetDescendantComponentsArray(Meshes.SM_Clock_Button)[0]
        then:
            set WhiteButtonMesh.Material = chess_side.White.Material()

        if:
            BlackButton := MaybeBlackReadyButton?
            BlackButtonMesh := BlackButton.GetDescendantComponentsArray(Meshes.SM_Clock_Button)[0]
        then:
            set BlackButtonMesh.Material = chess_side.Black.Material()

        if:
            RefreshButton := MaybeWhiteReadyButton?
            RefreshButtonMesh := RefreshButton.GetDescendantComponentsArray(Meshes.SM_Clock_Button)[0]
        then:
            set RefreshButtonMesh.Material = Materials.MI_RefreshButton{}

        if:
            PauseButton := MaybeWhiteReadyButton?
            PauseButtonMesh := PauseButton.GetDescendantComponentsArray(Meshes.SM_Clock_Button)[0]
        then:
            set PauseButtonMesh.Material = Materials.MI_PauseButton{}
            

    # Handles top-level game logic
    OnRoundStart()<suspends>:void=
        race:
            EndSimulationEvent.Await() 
            RemovedFromSceneEvent.Await()
            block:
                if:
                    Playspace := Entity.GetPlayspaceForEntity[]
                    ChessBoard := MaybeChessBoard?
                    BoardComp := ChessBoard.GetBoardComponent[]
                then:
                    Playspace.PlayerRemovedEvent().Subscribe(OnPlayerRemoved)

                    Print("chess_game_manager_component\n Both Players must ready up to start the game")

                    spawn. ListenRefreshButtonInteraction()
                    spawn. ListenReadyButtonInteraction(chess_side.White)
                    spawn. ListenReadyButtonInteraction(chess_side.Black)
                    spawn. ListenGameSpeedSelection()
                    
                    loop: 
                        set GameState = chess_game_state.PreGame
                        Sleep(1.0)
                        ResetGame()
                        BothPlayersReadyEvent.Await()
                                
                        # On Game Begin
                        GameBeginEvent.Signal()
                        set GameState = chess_game_state.GameInProgress
                        Print("chess_game_manager_component\n Game Begins!")
                        
                        MaybeWinner: ?chess_player_info = race:
                            block: # Turn Cycle Logic
                                loop:
                                    if (WhiteSideInfo.TurnsTaken = BlackSideInfo.TurnsTaken):
                                        set CurrentRound += 1
                                        Print("chess_game_manager_component\n Round {CurrentRound}!")

                                    PlayerTurnLogic(CurrentTurn)                                
                                    set CurrentTurn = CurrentTurn.Opponent()    # Give the opponent the next turn
                                    Sleep(0.0)
                                false
                            block:  # On Player Abandons or Concedes game
                                Winner := (PlayerConcedeEvent.Await()).ChessSide.GetOpponentInfo()
                                option. Winner
                            block:  # On Player Captures King
                                Winner := (BoardComp.KingCaptured.Await()).GetOpponentInfo()
                                option. Winner
                            block:  # On Player Turn Timer Elapses
                                Winner := race:
                                    block:
                                        WhiteSideInfo.TurnTimer.TimerExpireEvent.Await()
                                        chess_side.White.GetOpponentInfo()
                                    block:
                                        BlackSideInfo.TurnTimer.TimerExpireEvent.Await()
                                        chess_side.Black.GetOpponentInfo()
                                option. Winner
                        
                        # On Post Game
                        set GameState = chess_game_state.PostGame
                        if (Winner := MaybeWinner?):
                            # On Player win
                            Print("chess_game_manager_component\n {Winner.ChessSide} is the winner!")
                        else:
                            # On Draw
                            Print("chess_game_manager_component\n The game is a DRAW! No one wins")

                        GameOverEvent.Signal(MaybeWinner)
                        Print("chess_game_manager_component\n Game Manager Resetting in {PostGameDelay} seconds")
                        Sleep(PostGameDelay)

                        PostGameDelayCompleteEvent.Signal()
    

    OnEndSimulation<override>():void=
        (super:)OnEndSimulation()
        EndSimulationEvent.Signal()

    
    OnRemovingFromScene<override>():void=
        (super:)OnRemovingFromScene()
        RemovedFromSceneEvent.Signal()
    

    # Continuously handles interaction behavior for the Refresh button
    ListenRefreshButtonInteraction()<suspends>:void=
        race:
            RemovedFromSceneEvent.Await()
            EndSimulationEvent.Await()
            loop:
                if (ReadyButton := MaybeRefreshButton?.GetComponent[interactable_component]):
                    ReadyButton.SucceededEvent.Await()
                
                if:
                    ChessBoard := MaybeChessBoard?
                    BoardComp := ChessBoard.GetBoardComponent[]
                    BoardComp.isInFreeplayMode?
                then:
                    BoardComp.RefreshBoard()
                Sleep(0.0)


    # Continuously handles interaction behavior for the indicated chess_side
    ListenReadyButtonInteraction(Side: chess_side)<suspends>:void=
        race:
            RemovedFromSceneEvent.Await()
            EndSimulationEvent.Await()
            block:
                SideInfo := Side.GetInfo()
                if (Side.CanOverrideInteraction[]):
                    Print("chess_game_manager_component\n{Side}'s Ready Interaction Has Been Overridden")
                    PlayerReadyEvent.Signal(SideInfo)
                    Sleep(DebugStepDelay)
                    return
                
                # Determine which Ready Button belongs to this color
                MaybeReadyButton := case (Side):
                    chess_side.White => MaybeWhiteReadyButton
                    chess_side.Black => MaybeBlackReadyButton
                # Determine which Status Display belongs to this color
                MaybeStatusDisplay := case (Side):
                    chess_side.White => PlayerWhiteStatusDisplay
                    chess_side.Black => PlayerBlackStatusDisplay

                MaybeStatusDisplay.SetText(ClockSideReadyStatus(SideToMessage(Side), StringToMessage("isn't ready")))

                if (ReadyButton := MaybeReadyButton?.GetComponent[interactable_component]):
                    loop:  
                        Instigator: agent = ReadyButton.SucceededEvent.Await()
                        if (Player := player[Instigator]):  # cast Instigator to player since we'll need that regardless of what we do
                            
                             # The Ready button will do different things depending on the game state
                            case (GameState):

                                chess_game_state.PreGame =>
                                    if:
                                        not SideInfo.MaybePlayer?   # This side does not already have a player
                                        not (Side.GetOpponentInfo().MaybePlayer? and Player = Side.GetOpponentInfo().MaybePlayer?)  # The instigating player is not already on the opposing side
                                    then:
                                        set SideInfo.MaybePlayer = option. Player
                                        MaybeStatusDisplay.SetText(ClockSideReadyStatus(SideToMessage(Side), StringToMessage("is ready")))
                                        PlayerReadyEvent.Signal(SideInfo)
                                        Print("chess_game_manager_component\n {Side} is ready to begin")

                                        if(Side.GetOpponentInfo().MaybePlayer? and Side.GetInfo().MaybePlayer?): # Are both players ready
                                            BothPlayersReadyEvent.Signal()

                                chess_game_state.GameInProgress =>
                                    if:
                                        Player = SideInfo.MaybePlayer?  # a player must choose for themselves if they want to concede
                                        IsPlayersTurn[Player]             # player can only concede if its currently their turn
                                    then:
                                        OnPlayerRemoved(Player)
                                        MaybeStatusDisplay.SetText(ClockSideReadyStatus(SideToMessage(Side), StringToMessage("isn't ready")))
                                        PlayerUnreadyEvent.Signal(SideInfo)

                                chess_game_state.PostGame => Print("chess_game_manager_component\n{Side} Ready button has no functionality during post game")

                        Sleep(0.0)
                        
                else Print("chess_game_manager_component\n ERROR: Couldn't Await for {Side}'s Ready Interaction")


    # Continuously handles interaction behavior for the Game speed selection button
    ListenGameSpeedSelection()<suspends>:void=
        race:
            RemovedFromSceneEvent.Await()
            EndSimulationEvent.Await()
            block:
                if:
                    ModeSelectorButton := MaybeModeSelectButton?.GetComponent[interactable_component]
                    ChessBoard := MaybeChessBoard?
                    BoardComp := ChessBoard.GetBoardComponent[]     
                then:
                    loop:
                        Sleep(0.0)
                        ModeSelectorButton.SucceededEvent.Await()

                        # Only allow changing the timer before the game starts
                        if (GameState = chess_game_state.PreGame):
                            var Itr: int= GameSpeed.ToIndex() + 1
                            SpeedList := GetSpeedList()
                            # Cycling past the last speed mode should roll over back to the start
                            if (Itr >= SpeedList.Length):
                                set Itr = 0     
                            if (NextSpeed := SpeedList[Itr]):
                                set GameSpeed = NextSpeed

                            TurnTime := case (GameSpeed):
                                chess_game_speed.Quick => QuickGameTime
                                chess_game_speed.Long => LongGameTime
                                chess_game_speed.Infinite => 999999999999999.9
                                    
                            set WhiteSideInfo.TurnTimer.StartTime = TurnTime
                            set WhiteSideInfo.TurnTimer.TimeRemaining = TurnTime
                            set BlackSideInfo.TurnTimer.StartTime = TurnTime
                            set BlackSideInfo.TurnTimer.TimeRemaining = TurnTime

                            TimeClockDisplay.SetText(PreGameClockDisplayText(GameSpeedToMessage(GameSpeed)))


    # Handles turn logic for the given side
    PlayerTurnLogic(Side: chess_side)<suspends>:void=
        SideInfo := Side.GetInfo()
        TurnStartEvent.Signal(SideInfo)
        if (not GameSpeed = chess_game_speed.Infinite):
            SideInfo.TurnTimer.Play()
        
        # Await until player has made their decision or until their timer expires
        race:
            block: # On Turn Timer Elapsed
                SideInfo.TurnTimer.TimerExpireEvent.Await()
            loop: # Update Time Clock
                SideInfo.TurnTimer.TimerUpdateEvent.Await()
                TimeClockDisplay.SetText(TurnClockDisplayText(SideToMessage(Side), SideInfo.TurnTimer.ToMessage()))
                TimeClockDisplay.UpdateDisplay()
                Sleep(0.0)
            block:  # Debug Override Interaction
                if (not Side.CanOverrideInteraction[]):
                    Sleep(Inf)
                Print("chess_game_manager_component\n{Side}'s Turn Interaction Has Been Overridden")
                Sleep(DebugStepDelay)                
            loop:   # Await Player completing a move
                Sleep(0.0)
                if:
                    ChessBoard := MaybeChessBoard?
                    BoardComp := ChessBoard.GetBoardComponent[]
                then:
                    BoardComp.PieceMoveEndEvent.Await()
                    break

        # On Turn Complete
        SideInfo.TurnTimer.Pause()
        set SideInfo.TurnsTaken += 1
    
    
    OnPlayerRemoved(Player: player):void=
        if: 
            Player = WhiteSideInfo.MaybePlayer?
        then:
            PlayerConcedeEvent.Signal(WhiteSideInfo)
            set WhiteSideInfo.MaybePlayer = false
        else if: 
            Player = BlackSideInfo.MaybePlayer?
        then:
            PlayerConcedeEvent.Signal(BlackSideInfo)
            set BlackSideInfo.MaybePlayer = false


    # Resets all gameplay variables to their starting value 
    ResetGame():void=
        set CurrentTurn = chess_side.White
        set CurrentRound = 0

        if:
            ChessBoard := MaybeChessBoard?
            BoardComp := ChessBoard.GetBoardComponent[]
        then:
            BoardComp.RefreshBoard()  

        WhiteSideInfo.Reset()
        BlackSideInfo.Reset()

        TimeClockDisplay.SetText(PreGameClockDisplayText(GameSpeedToMessage(GameSpeed)))
        PlayerWhiteStatusDisplay.SetText(ClockSideReadyStatus(SideToMessage(chess_side.White), StringToMessage("isn't ready")))
        PlayerBlackStatusDisplay.SetText(ClockSideReadyStatus(SideToMessage(chess_side.Black), StringToMessage("isn't ready")))
        
        TimeClockDisplay.UpdateDisplay()
        PlayerWhiteStatusDisplay.UpdateDisplay()
        PlayerBlackStatusDisplay.UpdateDisplay()
        

    # Helper function that takes a chess_side variable and tells if it its debug override flag is set
    (Side: chess_side).CanOverrideInteraction()<transacts><decides>: void=
        case (Side):
            chess_side.White => WhiteInteractionOverride?
            chess_side.Black => BlackInteractionOverride?


    # Helper Function that converts a chess_side enum value to its corresponding chess_player_info object reference
    (Side: chess_side).GetInfo()<reads>:chess_player_info =
        case (Side):
            chess_side.White => WhiteSideInfo
            chess_side.Black => BlackSideInfo


    # Helper Function that converts a chess_side enum value to its opponent's chess_player_info object reference 
    (Side: chess_side).GetOpponentInfo()<reads>:chess_player_info = Side.Opponent().GetInfo()
    

GameCategory<public><localizes>:message := "Game Settings"
DebugCategory<public><localizes>:message := "Debug Settings"
PrefabSetupCategory<public><localizes>:message := "Entity Prefab Setup"

ClockSideReadyStatus<public><localizes>(SideMsg: message, ReadyMsg : message):message := "{SideMsg} {ReadyMsg}"
PreGameClockDisplayText<public><localizes>(GameSpeedString:message):message := "{GameSpeedString} Play"
TurnClockDisplayText<public><localizes>(SideMsg: message, FormatedTime:message):message := "{SideMsg}'s Turn\n{FormatedTime} "

###################### End Chess Game Manager Component #########################################################


###################### Start Chess Player Info ##################################################################

# Data structure that contains all relevent information that that player and their side of the board
chess_player_info<public> := class():
    var MaybePlayer<public>: ?player = false
    ChessSide<public>: chess_side
    TurnTimer<public>: chess_turn_timer = chess_turn_timer{}

    # If this value ever meets or exceeds 50, then its a draw
    var MovesSinceLastCapture<public>: int = 0

    # Stats
    var TurnsTaken<public>: int = 0

    Reset():void=
        set MaybePlayer = false
        set TurnsTaken = 0
        
        TurnTimer.Reset()


###################### End Chess Player Info #########################################################################


###################### Start Game Speed Enum + Helpers ############################################################

chess_game_speed<public> := enum {Quick, Long, Infinite}
(Speed: chess_game_speed).ToIndex():int=
    case (Speed):
        chess_game_speed.Quick => 0
        chess_game_speed.Long => 1
        chess_game_speed.Infinite => 2

ToString(Speed: chess_game_speed)<computes>:string =
    case (Speed):
        chess_game_speed.Quick => "Quick"
        chess_game_speed.Long => "Long"
        chess_game_speed.Infinite => "Infinite"

GameSpeedToMessage<public>(GameSpeed: chess_game_speed):message = StringToMessage(ToString(GameSpeed))

GetSpeedList()<computes>:[]chess_game_speed = array{chess_game_speed.Quick, chess_game_speed.Long, chess_game_speed.Infinite}
