<#
@UUID: ead44cb4-3904-4080-b066-ec9bc545ce99
@Task: Chess Piece Components
@Description: An abstract parent chess piece component that handles all the generic player interaction needed for all chess_pieces to funciton as well as a component for  every chess piece that inherits from the abstract base class and defines the unique movement behavior of each piece.
@Tags: SceneGraph, Gameplay, PlayerInteraction, Movement
@Export: true
#>

<#
In-Editor Entity Prefab Setup:
    (NOTE: any setting not mentioned is assumed to be default or to your preference) 

chess_piece_[PIECE NAME]
    RootEntity
        - ANY COMPONENT THAT INHERITS FROM "chess_piece_component"
        - mesh_component (using custom SM_Chess_[PIECE NAME] static mesh)
            - Material: M_ChessPiece_White (will change programatically when board is generated)
            - Collidable: false (must be disabled to allow for interactable collider)
            - Queryable: false
        - transform_component
        PieceInteractCollider
            - interactable_component
            - mesh_component (preferably a cylinder)
            - transform_component
                - Visible: false
                - Tweak the scale values so that it is just slightly large enough to envelop RootEntity's mesh component

    (NOTE: this entity prefab definition applies to all the pieces, just replace [PIECE NAME] with whatever the piece is.)
#>

using { /Verse.org }
using { /Verse.org/Native }
using { /Verse.org/SceneGraph }
using { /Verse.org/Simulation }
using { /Verse.org/SpatialMath }
using { /Verse.org/Assets }
using { EPs }

############## Base Chess Piece ####################    

chess_piece_component<public> := class<final_super><abstract>(component):

    @editable_slider(float):
        MinValue := option. 0.0
    MoveSpeed: float = 10.0

    var MaybeBoard:?chess_board_component = false
    var MaybeTile:?chess_tile_component = false
    var StartingRow : int = -1
    var StartingColumn : int = -1
    var Side : chess_side = chess_side.White
    var Selected : logic = false
    var Targeted : logic = false
    
    # Abstract functions to be overridden by individual piece components
    GetTargetablePositions() : []chess_tile_component # Returns positions this piece can currently reach
    SetMaterial(NewMaterial : material) : void

    
    OnSimulate<override>()<suspends>:void=
        if (ChessPieceInteract := Entity.GetDescendantComponentsArray(interactable_component)[0]):
            loop:
                Agent : agent = ChessPieceInteract.SucceededEvent.Await()
                if(Player := player[Agent]):
                    spawn. Interaction(Player)
                    Sleep(0.0)
    

    InitPiece(Board : chess_board_component, StartRow : int, StartColumn : int, BoardSide : chess_side) : void=
        set MaybeBoard = option. Board
        set StartingRow = StartRow
        set StartingColumn = StartColumn
        set Side = BoardSide
        if(Tile := Board.GetTile[StartingRow, StartingColumn]):
            set MaybeTile = option. Tile
        SetMaterial(Side.Material())


    # Highlight positions (and enable their interactable comp) this piece can move to
    Interaction(Player : player)<suspends> : void=
        if (Board := MaybeBoard?):
            if (Board.isInFreeplayMode?):
                spawn. Board.PieceInteractedWith(Self)
            else if:
                GameManager := Board.MaybeGameManager?
                GameManager.CurrentTurn = Side or Targeted?
                GameManager.IsPlayersTurn[Player] or Targeted?
            then:
                spawn. Board.PieceInteractedWith(Self)


    TargetPiece() : void=
        set Targeted = true
        SetMaterial(Side.GlowMaterial())


    UnTargetPiece() : void=
        set Targeted = false
        SetMaterial(Side.Material())


    SelectPiece()<suspends> : void=
        set Selected = true
        SetMaterial(Side.GlowMaterial())
        RaisePiece()


    DeselectPiece()<suspends> : void=
        set Selected = false
        SetMaterial(Side.Material())
        LowerPiece()


    RaisePiece()<suspends> : void=
        if(Board := MaybeBoard?, Tile := MaybeTile?):
            # Raise piece in the air to indicate it's been selected
            ChangeHeight(Self, Tile.Entity.GetLocalTransform().Translation, ?Height := 75.0)
    

    LowerPiece()<suspends> : void=
        if(Board := MaybeBoard?, Tile := MaybeTile?):
            # Lower piece in the air to indicate it's been deselected
            ChangeHeight(Self, Tile.Entity.GetLocalTransform().Translation)


    ChangeHeight(Piece : chess_piece_component, TileBasePos : (/Verse.org/SpatialMath:) vector3, ?Height : float = 0.0)<suspends>:void=
        if(Board := MaybeBoard?):
            race:
                block:
                    PieceTransform := Piece.Entity.GetLocalTransform()
                    StartPos := PieceTransform.Translation
                    # Adding 12.5 to the "up" value sets the piece on top of the tile. This should scale with the size of the tiles.
                    EndPos := (/Verse.org/SpatialMath:) vector3{
                        Forward := TileBasePos.Forward, 
                        Left := TileBasePos.Left, 
                        Up := TileBasePos.Up + (12.5*Board.BoardScale) + Height
                    }
                    
                    Timestamp := GetSimulationElapsedTime()
                    loop:
                        Sleep(0.0)
                        DeltaTime := GetSimulationElapsedTime() - Timestamp
                        Interp := Clamp(Piece.MoveSpeed*DeltaTime, 0.0, 1.0)

                        NewTransform :=  (/Verse.org/SpatialMath:)transform{
                            Translation := Lerp(StartPos, EndPos, Interp),
                            Rotation := PieceTransform.Rotation,
                            Scale := PieceTransform.Scale
                        }
                        Piece.Entity.SetLocalTransform(NewTransform)

                        if(Interp = 1.0):
                            break
                block:
                    Board.EndSimulationEvent.Await()
    

    CheckTilesInLine(RowDirection : int, ColumnDirection : int) : []chess_tile_component=
        var Result : []chess_tile_component = array{}
        if(Board := MaybeBoard?, Tile := MaybeTile?):
            var RowIndex : int = Tile.Row
            var ColumnIndex :int = Tile.Column
            loop:
                set RowIndex += RowDirection
                set ColumnIndex += ColumnDirection

                if(TargetTile := Board.GetTile[RowIndex, ColumnIndex]):
                    if(Piece := TargetTile.MaybeChessPieceOnTile?):
                        if(Piece.Side = Side):
                            break:
                        else:
                            set Result += array{TargetTile}
                            break
                    else:
                        set Result += array{TargetTile}
                else: 
                    break

        Result

############## End Base Chess Piece ####################    


############## Individual Pieces ####################    

chess_pawn_component<public> := class(chess_piece_component):

    var MovedOnce<public> : logic = false

    SetMaterial<override>(NewMaterial : material) : void=
        if(thisStaticMesh := Entity.GetComponent[Meshes.SM_Chess_Pawn]):
            set thisStaticMesh.Material = NewMaterial

    # If in original position, return 2 spaces in front
    # otherwise return 1 space in front
    # If front-left or front-right space is occupied, add those spaces
    GetTargetablePositions<override>() : []chess_tile_component=
        var Result : []chess_tile_component = array{}

        if(Board := MaybeBoard?, Tile := MaybeTile?):
            # Pawn moves 1 space forward. Pawns cannot move through pieces, so targeting new spaces are nested in targeting the first spaces.
            if:
                ForwardTile := Board.GetTile[Tile.Row + Side.ToRowDirection(), Tile.Column]
                ForwardTile.MaybeChessPieceOnTile = false
            then:
                set Result += array{ForwardTile}
            
                # If this is the first time a pawn is moving, it can move an extra space forward.
                if:
                    not MovedOnce?
                    ForwardTile2 := Board.GetTile[Tile.Row + Side.ToRowDirection()*2, Tile.Column]
                    ForwardTile2.MaybeChessPieceOnTile = false
                then:
                    set Result += array{ForwardTile2}
            
            # Pawn can only target certain spaces if there's a piece to capture there
            if:
                ForwardTile := Board.GetTile[Tile.Row + Side.ToRowDirection(), Tile.Column + Side.ToRowDirection()]
                ChessPiece := ForwardTile.MaybeChessPieceOnTile?
                ChessPiece.Side <> Self.Side
            then:
                set Result += array{ForwardTile}

            if:
                ForwardTile := Board.GetTile[Tile.Row + Side.ToRowDirection(), Tile.Column - Side.ToRowDirection()]
                ChessPiece := ForwardTile.MaybeChessPieceOnTile?
                ChessPiece.Side <> Self.Side
            then:
                set Result += array{ForwardTile}
        
        Result


    Upgrade<public>():void=
        if(Board := MaybeBoard?, Tile := MaybeTile?):
            Entity.RemoveFromParent()
            NewQueen := chess_piece_queen{}
            if(NewQueenComp := NewQueen.GetComponent[chess_queen_component]):
                NewQueenComp.InitPiece(Board, Tile.Row, Tile.Column, Side)
                Board.AddPieceToBoard(NewQueenComp)
                set Board.Pieces += array{NewQueenComp}


    
chess_rook_component<public> := class(chess_piece_component):

    SetMaterial<override>(NewMaterial : material) : void=
        if(thisStaticMesh := Entity.GetComponent[Meshes.SM_Chess_Rook]):
            set thisStaticMesh.Material = NewMaterial

    #return current row and column until edge of board
    GetTargetablePositions<override>() : []chess_tile_component=
        var Result : []chess_tile_component = array{}

        set Result += CheckTilesInLine(1, 0)#Check Right
        set Result += CheckTilesInLine(-1, 0)#Check Left
        set Result += CheckTilesInLine(0, 1)#Check Forward
        set Result += CheckTilesInLine(0, -1)#Check Backward

        Result



chess_bishop_component<public> := class(chess_piece_component):

    SetMaterial<override>(NewMaterial : material) : void=
        if(thisStaticMesh := Entity.GetComponent[Meshes.SM_Chess_Bishop]):
            set thisStaticMesh.Material = NewMaterial

    #return diagonals until edge of board
    GetTargetablePositions<override>() : []chess_tile_component=
        var Result : []chess_tile_component = array{}

        set Result += CheckTilesInLine(1,1)#Forward Right
        set Result += CheckTilesInLine(-1,1)#Forward Left
        set Result += CheckTilesInLine(1,-1)#Backward Right
        set Result += CheckTilesInLine(-1,-1)#Backward Left

        Result



chess_knight_component<public> := class(chess_piece_component):

    SetMaterial<override>(NewMaterial : material) : void=
        if(thisStaticMesh := Entity.GetComponent[Meshes.SM_Chess_Knight]):
            set thisStaticMesh.Material = NewMaterial

    #return 8 spaces, +/-2over +/-1up, +/-2up +/-1over
    GetTargetablePositions<override>() : []chess_tile_component=
        var Result : []chess_tile_component = array{}

        if(Board := MaybeBoard?, Tile := MaybeTile?):
            # Upper right spaces
            if:
                TargetTile := Board.GetTile[Tile.Row + 1, Tile.Column + 2]
                TargetTile.MaybeChessPieceOnTile = false or (TargetTile.MaybeChessPieceOnTile?.Side <> Self.Side)
            then:
                set Result += array{TargetTile}

            if:
                TargetTile := Board.GetTile[Tile.Row + 2, Tile.Column + 1]
                TargetTile.MaybeChessPieceOnTile = false or (TargetTile.MaybeChessPieceOnTile?.Side <> Self.Side)
            then:
                set Result += array{TargetTile}

            # Lower left spaces
            if:
                TargetTile := Board.GetTile[Tile.Row - 1, Tile.Column - 2]
                TargetTile.MaybeChessPieceOnTile = false or (TargetTile.MaybeChessPieceOnTile?.Side <> Self.Side)
            then:
                set Result += array{TargetTile}

            if:
                TargetTile := Board.GetTile[Tile.Row - 2, Tile.Column - 1]
                TargetTile.MaybeChessPieceOnTile = false or (TargetTile.MaybeChessPieceOnTile?.Side <> Self.Side)
            then:
                set Result += array{TargetTile}

            # Lower right spaces
            if:
                TargetTile := Board.GetTile[Tile.Row - 1, Tile.Column + 2]
                TargetTile.MaybeChessPieceOnTile = false or (TargetTile.MaybeChessPieceOnTile?.Side <> Self.Side)
            then:
                set Result += array{TargetTile}

            if:
                TargetTile := Board.GetTile[Tile.Row - 2, Tile.Column + 1]
                TargetTile.MaybeChessPieceOnTile = false or (TargetTile.MaybeChessPieceOnTile?.Side <> Self.Side)
            then:
                set Result += array{TargetTile}

            # Upper left spaces
            if:
                TargetTile := Board.GetTile[Tile.Row + 1, Tile.Column - 2]
                TargetTile.MaybeChessPieceOnTile = false or (TargetTile.MaybeChessPieceOnTile?.Side <> Self.Side)
            then:
                set Result += array{TargetTile}

            if:
                TargetTile := Board.GetTile[Tile.Row + 2, Tile.Column - 1]
                TargetTile.MaybeChessPieceOnTile = false or (TargetTile.MaybeChessPieceOnTile?.Side <> Self.Side)
            then:
                set Result += array{TargetTile}
        
        Result



chess_queen_component<public> := class(chess_piece_component):

    SetMaterial<override>(NewMaterial : material) : void=
        if(thisStaticMesh := Entity.GetComponent[Meshes.SM_Chess_Queen]):
            set thisStaticMesh.Material = NewMaterial

    #return row column and diagonals to edge of board
    GetTargetablePositions<override>() : []chess_tile_component=
        var Result : []chess_tile_component = array{}

        set Result += CheckTilesInLine(1, 0)#Check Right
        set Result += CheckTilesInLine(-1, 0)#Check Left
        set Result += CheckTilesInLine(0, 1)#Check Forward
        set Result += CheckTilesInLine(0, -1)#Check Backward
        set Result += CheckTilesInLine(1, 1)#Forward Right
        set Result += CheckTilesInLine(-1, 1)#Forward Left
        set Result += CheckTilesInLine(1, -1)#Backward Right
        set Result += CheckTilesInLine(-1, -1)#Backward Left
        
        Result



chess_king_component<public> := class(chess_piece_component):

    SetMaterial<override>(NewMaterial : material) : void=
        if(thisStaticMesh := Entity.GetComponent[Meshes.SM_Chess_King]):
            set thisStaticMesh.Material = NewMaterial

    #return 8 neighboring spaces    
    GetTargetablePositions<override>() : []chess_tile_component=
        var Result : []chess_tile_component = array{}

        if(Board := MaybeBoard?, Tile := MaybeTile?):
            # Up Right
            if:
                TargetTile := Board.GetTile[Tile.Row + 1, Tile.Column + 1]
                TargetTile.MaybeChessPieceOnTile = false or (TargetTile.MaybeChessPieceOnTile?.Side <> Self.Side)
            then:
                set Result += array{TargetTile}

            # Up
            if:
                TargetTile := Board.GetTile[Tile.Row, Tile.Column + 1]
                TargetTile.MaybeChessPieceOnTile = false or (TargetTile.MaybeChessPieceOnTile?.Side <> Self.Side)
            then:
                set Result += array{TargetTile}

            # Up Left
            if:
                TargetTile := Board.GetTile[Tile.Row + 1, Tile.Column - 1]
                TargetTile.MaybeChessPieceOnTile = false or (TargetTile.MaybeChessPieceOnTile?.Side <> Self.Side)
            then:
                set Result += array{TargetTile}

            # Left
            if:
                TargetTile := Board.GetTile[Tile.Row, Tile.Column - 1]
                TargetTile.MaybeChessPieceOnTile = false or (TargetTile.MaybeChessPieceOnTile?.Side <> Self.Side)
            then:
                set Result += array{TargetTile}

            # Down Left
            if:
                TargetTile := Board.GetTile[Tile.Row - 1, Tile.Column - 1]
                TargetTile.MaybeChessPieceOnTile = false or (TargetTile.MaybeChessPieceOnTile?.Side <> Self.Side)
            then:
                set Result += array{TargetTile}

            # Down
            if:
                TargetTile := Board.GetTile[Tile.Row - 1, Tile.Column]
                TargetTile.MaybeChessPieceOnTile = false or (TargetTile.MaybeChessPieceOnTile?.Side <> Self.Side)
            then:
                set Result += array{TargetTile}

            # Down Right
            if:
                TargetTile := Board.GetTile[Tile.Row - 1, Tile.Column + 1]
                TargetTile.MaybeChessPieceOnTile = false or (TargetTile.MaybeChessPieceOnTile?.Side <> Self.Side)
            then:
                set Result += array{TargetTile}

            # Right
            if:
                TargetTile := Board.GetTile[Tile.Row + 1, Tile.Column]
                TargetTile.MaybeChessPieceOnTile = false or (TargetTile.MaybeChessPieceOnTile?.Side <> Self.Side)
            then:
                set Result += array{TargetTile}
        
        Result

############## End Individual Pieces ####################    