<#
@UUID: ead44cb4-3904-4080-b066-ec9bc545ce99
@Task: Chess Board Tile Component
@Description: Handles the appearance and interactibility of an individual chess board tile
@Tags: Gameplay, SceneGraph, PlayerInteraction
@Export: true
#>

<#
In-Editor Entity Prefab Setup:
    (NOTE: any setting not mentioned is assumed to be default or to your preference)

chess_board_tile
    RootEntity
        - mesh_component (using custom SM_Chess_Tile static mesh)
            - Material: M_ChessPiece_Black (will change programatically when board is generated)
            - Collidable: false (must be disabled to allow for interactable collider)
            - Queryable: false
        - transform_component
        TileInteractCollider
            - interactable_component
            - mesh_component (preferably a cube)
            - transform_component
                - Visible: false
                - Tweak the scale values so that it is just slightly large enough to envelop RootEntity's mesh component
#>

using { /Verse.org }
using { /Verse.org/Native }
using { /Verse.org/SceneGraph }
using { /Verse.org/Simulation }
using { /Verse.org/Assets }
using { EPs }


chess_tile_component<public> := class<final_super>(component):

    var MaybeBoard:?chess_board_component = false
    var MaybeInteractable: ?interactable_component = false
    var Row: int = 0
    var Column: int = 0
    var Color: chess_side = chess_side.White
    var MaybeChessPieceOnTile : ?chess_piece_component = false
    var IsHighlighted: logic = false
    
    OnBeginSimulation<override>():void=
        (super:)OnBeginSimulation()

        # Get Reference to interaction component on this entity (assumes there's only one)
        set MaybeInteractable = option. Entity.GetDescendantComponentsArray(interactable_component)[0]

        if (Interactable := MaybeInteractable?):
            Interactable.SucceededEvent.Subscribe(OnInteract)
        else:
            Print("chess_tile_component:\nERROR: couldn't find interactable component on or descended from this entity")

    
    InitTile(Board : chess_board_component, inRow : int, inColumn : int, inColor : chess_side) : void=
        set MaybeBoard = option. Board
        set Row = inRow
        set Column = inColumn
        set Color = inColor
        SetMaterial(Color.Material())


    OnInteract(Instigator: agent):void=
        if:
            Board := MaybeBoard?
            SelectedPiece := Board.MaybeSelectedPiece?
            ThisGridSlot := Board.ChessBoard[Row][Column]
        then:
            spawn. Board.MovePiece(SelectedPiece, ThisGridSlot)
    
        
    Highlight():void =
        if:
            not IsHighlighted?
            Interactable := MaybeInteractable?   
        then:
            set IsHighlighted = true
            SetMaterial(Color.GlowMaterial())
            Interactable.Enable()

            if(Piece := MaybeChessPieceOnTile?):
                Piece.TargetPiece()


    Unhighlight():void =
         if:
            IsHighlighted?
            Interactable := MaybeInteractable?        
        then:
            set IsHighlighted = false
            SetMaterial(Color.Material())
            Interactable.Disable()
            
            if(Piece := MaybeChessPieceOnTile?):
                Piece.UnTargetPiece()
            

    SetMaterial(NewMaterial : material): void= 
        if(Mesh := Entity.GetComponent[Meshes.SM_Chess_Tile]):
            set Mesh.Material = NewMaterial
