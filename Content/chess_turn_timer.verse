<#
@UUID: ead44cb4-3904-4080-b066-ec9bc545ce99
@Task: Chess Turn Timer Class
@Description: Utility class used primarily by the chess_game_manager_component to track the time each player has remaining until they lose the game. When the timer starts, it asyncronously ticks down and fires updates about its status. It can also be paused or reset at any time.
@Tags: Utility, DataStructures, TimerManagement
@Export: true
#>

using { /Verse.org }
using { /Verse.org/Concurrency }
using { /Verse.org/Simulation }
using { /Verse.org/Native }
using { /UnrealEngine.com/Temporary/SpatialMath }
using { /UnrealEngine.com/Temporary/Diagnostics }

timer_state := enum{Idle,Playing,Paused}

chess_turn_timer<public>:= class<concrete>():
    # ============ INTERNAL VARS ====================================

    var StartTime: float = 360.0
    var TimeRemaining: float = 360.0
    UpdateRate: float = 1.0

    var State: timer_state = timer_state.Idle

    # ============ EVENTS =======================================
    # Signals when the timer begins ticking down from its 
    TimerStartEvent<public> : event() = event(){}
    # Signals when the game timer has elapsed but before a winner is determined by the game manager component
    TimerExpireEvent<public> : event() = event(){}
    # Signals every time the TimeRemaining value is updated (countdown and game timer), passes the current remaining time
    TimerUpdateEvent<public> : event(float) = event(float){}
    # Signals when timer goes from playing to being paused.
    PauseEvent<public> : event() = event(){}
    # Signals when timer goes from being paused to being playing again.
    ResumeEvent<public> : event() = event(){}
    # Signals when the Timer's Reset method is called
    ResetEvent<public> : event() = event(){}


    Play<public>():void=
        case (State):
            timer_state.Playing => return
            timer_state.Idle => 
                set TimeRemaining = StartTime
                Print("chess_turn_timer\nTimer Started!")
                TimerStartEvent.Signal()
            timer_state.Paused => 
                Print("chess_turn_timer\nTimer Resuming")
                ResumeEvent.Signal()
        
        set State = timer_state.Playing
        spawn. TimerTick()
        

    Pause<public>():void=
        if (State = timer_state.Playing):
            set State = timer_state.Paused
            PauseEvent.Signal()

            
    # Stops the current timer and resets Time Remaining to its starting value
    Reset<public>():void =
        ResetEvent.Signal()
        set TimeRemaining = StartTime
        set State = timer_state.Idle
    
        
    TimerTick<private>()<suspends>:void=
        Print("chess_turn_timer\nTime Remaining {TimeRemaining}")
        race:
            block: # On Timer Reset
                ResetEvent.Await()
                return
            block:  # On Timer Paused
                PauseEvent.Await()
                return
            loop:
                # Calculate the delta between frames
                Timestamp := GetSimulationElapsedTime()
                Sleep(UpdateRate)
                DeltaTime := GetSimulationElapsedTime() - Timestamp

                set TimeRemaining -= DeltaTime  # Update the RemainingTime
                TimerUpdateEvent.Signal(TimeRemaining)

                if (TimeRemaining <= 0.0):
                    break
        
        # On Timer Expired
        set TimeRemaining = 0.0
        set State = timer_state.Idle
        TimerExpireEvent.Signal()


    ToMessage<public>():message =
        TimeToMessage(MakeTimeFromSeconds(TimeRemaining))