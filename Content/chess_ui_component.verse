<#
@UUID: ead44cb4-3904-4080-b066-ec9bc545ce99
@Task: Chess Game HUD Component
@Description: Handles the pure-Verse UI of the Chess Minigame. Reacts to game manager events to update the team's scores and display a timer and the game's result.
@Tags: UI, SceneGraph, DataStructures, Formatting
@Export: true
#>

<#
In-Editor Entity Prefab Setup:
    (NOTE: any setting not mentioned is assumed to be default or to your preference)

chess_board
    RootEntity
        - chess_board_component
        - PieceMove (sound_component)
            - Sounds: populate array with all the chess piece move sounds you want/need
        - transform_component

chess_time_clock
        RootEntity
            - mesh_component (using the SM_Clock_Body static mesh)
                - Material: M_Clock_Body
                - Material_001: M_Clock_Screen
            - transofrm_component

            ModeSelectionButton:
                - interactable_component
                - mesh_component (using the SM_button static mesh)
                    - WorldGridMaterial: M_PauseButton
                - transform_component
                    - tweak transoform until the button is poking out in the center below the clock screen dispaly

            RefreshButton:
                - interactable_component
                - mesh_component (using the SM_button static mesh)
                    - WorldGridMaterial: M_RefreshButton
                - transform_component
                    - tweak transoform until the button is poking out on the outer edge of the clock face panel
            
            WhiteSideReadyButton:
                - interactable_component
                - mesh_component (using the SM_button static mesh)
                    - WorldGridMaterial: M_ChessPiece_White
                - transform_component
                    - tweak transoform until the button is poking out to the right of the ModeSelectionButton
            
            BlackSideReadyButton:
                - interactable_component
                - mesh_component (using the SM_button static mesh)
                    - WorldGridMaterial: M_ChessPiece_Black
                - transform_component
                    - tweak transoform until the button is poking out to the left of the ModeSelectionButton

chess_game_template
    RootEntity
        - chess_game_manager_component
            - MaybeChessBoardEntity:
                - MaybeModeSelectionButton: set to "ModeSelectionButton"
                - MaybeWhiteReadyButton: set to "WhiteSideReadyButton"
                - MaybeBlackReadyButton: set to "BlackSideReadyButton"
                - MaybeRefreshButton: set to "RefreshButton"
                - TimeClockDisplay: billboard_device in the level 
                - PlayerWhiteStatusDisplay: billboard_device in the level 
                - PlayerBlackStatusDisplay: billboard_device in the level 
        - chess_hud_component

        ChessBoard (chess_board entity prefab)
            - transform_component
                - position at origin of Root Entity

        TimeClock (chess_time_clock entity prefab)
            - transform_component
                - position facing the chess_board compoennt and back far enough that it would not intersect with board when it generates (might require tiral and error)
        

#>


using { /UnrealEngine.com/Temporary/UI }
using { /UnrealEngine.com/Temporary/SpatialMath }
using { /Fortnite.com/Game }
using { /Fortnite.com/UI }
using { /Fortnite.com/Teams }
using { /Fortnite.com/Playspaces }
using { /Verse.org }    
using { /Verse.org/Colors }
using { /Verse.org/Native }
using { /Verse.org/SceneGraph }
using { /Verse.org/Simulation }


###################### Start UI Component ##########################################################################

chess_hud_component<public> := class<final_super>(component):

    var PlayerToUI:[player]chess_hud = map{} # Stores the UI object for each player
    var MaybeGameManager: ?chess_game_manager_component = false  # Reference to the sibling chess_game_manager_component

    # Signals when this component's OnEndSimulation function fires; used to cancel asyncronous logic when this component stops simulating
    EndSimulationEvent<public>: event() = event(){}
    # Signals when this component's OnRemovedFromScene function fires; used to cancel asyncronous logic when this component stops simulating
    RemovedFromSceneEvent<public>: event() = event(){}
    
    OnBeginSimulation<override>():void =
        (super:)OnBeginSimulation()

        # Get Reference to Sibling Game Manager Component and setup gameplay callbacks
        set MaybeGameManager = option. Entity.GetComponent[chess_game_manager_component] 

        if (Playspace := Entity.GetPlayspaceForEntity[]):
            Playspace.PlayerRemovedEvent().Subscribe(OnPlayerRemoved)
            Playspace.PlayerRemovedEvent().Subscribe(OnPlayerRemoved)

        if (RoundMan := Entity.GetFortRoundManager[]):
            RoundMan.SubscribeRoundStarted(OnRoundStart)


    OnRoundStart()<suspends>:void=
        race:
            EndSimulationEvent.Await()
            RemovedFromSceneEvent.Await()
            block:
                if (GameManager := MaybeGameManager?):
                    
                    # Add UI to every player's screen when the initializer device initilaizes the GameManager Component
                    AddUIAllPlayers()
                    loop:
                        # Waiting For Player Ready
                        UpdateTurnStatusText(InstructionMessage())
                        race:
                            GameManager.GameBeginEvent.Await()
                            loop: # On Player Ready
                                SideInfo := GameManager.PlayerReadyEvent.Await()
                                ShowPlayerReady(SideInfo.ChessSide)
                                Sleep(0.0)
                            loop: # On Player Unready
                                SideInfo := GameManager.PlayerUnreadyEvent.Await()
                                ShowPlayerUnready(SideInfo.ChessSide)
                                Sleep(0.0)
                            loop: # On Game Speed Change
                                NewSpeed := GameManager.GameSpeedChangeEvent.Await()
                                Sleep(0.0)
                                
                        UpdateTurnTimerText(GameManager.WhiteSideInfo)
                        UpdateTurnTimerText(GameManager.BlackSideInfo)

                        # Game Loop
                        var MaybeWinner: ?chess_player_info = false
                        race:
                            set MaybeWinner = GameManager.GameOverEvent.Await()
                            AwaitTurnTimerUpdate(GameManager.WhiteSideInfo)
                            AwaitTurnTimerUpdate(GameManager.BlackSideInfo)
                            loop:   # On Turn Start
                                CurrentTurnInfo := GameManager.TurnStartEvent.Await()
                                NewColor := CurrentTurnInfo.ChessSide.ToColor()
                                UpdateTurnStatusText(PlayerTurnStartMessage(ToString(CurrentTurnInfo.ChessSide)), ?MaybeNewColor := option. NewColor)
                                Sleep(0.0)
                            
                        # Post Game
                        UpdateTurnStatusText(EndGameMessage(), ?MaybeNewColor := option. NamedColors.Yellow)
                        ResultMessage := if (Winner := MaybeWinner?):
                            PlayerWinMessage(ToString(Winner.ChessSide))
                        else:
                            DrawMessage()
                        UpdateResultText(ResultMessage)
                        
                        GameManager.PostGameDelayCompleteEvent.Await()
                        ResetUI()
                
                    
    OnEndSimulation<override>():void =
        (super:)OnEndSimulation()
        RemoveUIAllPlayers()    # Remove all UI elements on end; Custom Verse UI does not get cleaned up automatically.
        EndSimulationEvent.Signal()

        
    OnRemovingFromScene<override>():void =
        (super:)OnRemovingFromScene()
        RemovedFromSceneEvent.Signal()


    ShowPlayerReady(Side: chess_side) : void = 
        for (Player->Hud : PlayerToUI):
            Hud.GetSideTimerText(Side).SetText(PlayerReadyMessage(ToString(Side)))

    
    ShowPlayerUnready(Side: chess_side) : void = 
        for (Player->Hud : PlayerToUI):
            case (Side):
                chess_side.White => Hud.WhiteTimerText.SetText(DefaultTimerMessage())
                chess_side.Black => Hud.BlackTimerText.SetText(DefaultTimerMessage())


    AwaitTurnTimerUpdate(SideInfo: chess_player_info)<suspends>:void =
        loop:
            SideInfo.TurnTimer.TimerUpdateEvent.Await()
            UpdateTurnTimerText(SideInfo)
            Sleep(0.0)


    UpdateTurnTimerText(SideInfo: chess_player_info):void =
        FormattedTime := SideInfo.TurnTimer.ToMessage()
            for (Player->Hud : PlayerToUI):
                Hud.GetSideTimerText(SideInfo.ChessSide).SetText(FormattedTime)


    UpdateTurnStatusText(Message:message,?MaybeNewColor:?color = false):void = 
        for (Player->Hud : PlayerToUI, StatusText := Hud.TurnStatusText):
            StatusText.SetText(Message)
            # Update text color if one was specified
            if (NewColor := MaybeNewColor?):
                StatusText.SetTextColor(NewColor)
                

    UpdateResultText(Message:message):void = 
        for (Player->Hud : PlayerToUI):
            Hud.ResultText.SetText(Message)

            
    ResetUI():void = 
        for (Player->Hud : PlayerToUI):
            Hud.Reset()


    AddUIAllPlayers():void =
        if(Playspace := Entity.GetPlayspaceForEntity[]):
            for (Player : Playspace.GetPlayers()) do Player.AddUI()


    RemoveUIAllPlayers():void=
        if(Playspace := Entity.GetPlayspaceForEntity[]):
            for(Player : Playspace.GetPlayers()):
                Player.RemoveUI()
            set PlayerToUI = map{}
            

    (Player : player).AddUI():void =
        if:
            PlayerUI := GetPlayerUI[Player]
            BattingGameUI := chess_hud{}
            set PlayerToUI[Player] = BattingGameUI
        then:
            BattingGameUI.CreateUI()
            PlayerUI.AddWidget(BattingGameUI.GetWidget())
        else:
            Print("chess_hud_component:\nERROR: Could not add UI to player's screen")
        

    (Player : player).RemoveUI():void =
        if:
            PlayerUI := GetPlayerUI[Player]
            BattingGameUI := PlayerToUI[Player]
        then:
            PlayerUI.RemoveWidget(BattingGameUI.GetWidget())
            set PlayerToUI = PlayerToUI.RemoveKey(Player)
    

    OnPlayerAdded(Player: player):void = Player.AddUI()
    OnPlayerRemoved(Player: player):void = Player.RemoveUI()

###################### End UI Component #############################################################################


###################### Start UI Display Messages ######################################################################

InstructionMessage<localizes>() : message = "Each Player Press A Button To Begin!"
EndGameMessage<localizes>() : message = "Done!"
DefaultTimerMessage<localizes>() : message = "Waiting"
PlayerReadyMessage<localizes>(SideString: string) : message = "{SideString} is Ready!"
PlayerTurnStartMessage<localizes>(SideString: string) : message = "{SideString}'s Turn."
PlayerWinMessage<localizes>(SideString: string) : message = "{SideString} is the victor!"
DrawMessage<localizes>() : message = "The game is a Draw!"

###################### End UI Display Messages ##########################################################################


###################### Start Player UI Class ############################################################################
                
chess_hud := class:
    
    WhiteTimerText: text_block = text_block{DefaultTextColor := NamedColors.White}
    BlackTimerText: text_block = text_block{DefaultTextColor := NamedColors.Black}

    TurnStatusText : text_block = text_block{DefaultTextColor := NamedColors.Yellow}
    ResultText : text_block = text_block{DefaultTextColor := NamedColors.Yellow}

    var MainWidget : canvas = canvas{}
    GetWidget() : canvas = MainWidget

    
    GetSideTimerText(Side: chess_side)<computes>: text_block = 
        case (Side):
            chess_side.White => WhiteTimerText
            chess_side.Black => BlackTimerText


    Reset() : void = 
        TurnStatusText.SetText(InstructionMessage())
        ResultText.SetText(EmptyMessage())
        WhiteTimerText.SetText(DefaultTimerMessage())
        BlackTimerText.SetText(DefaultTimerMessage())
    

    CreateUI() : void =
        #Displays: Team 1's Score    Team 2's Score
        ScoreUI := overlay:
            Slots := array:
                overlay_slot:
                    Widget := color_block:
                        DefaultColor := NamedColors.DarkGreen
                        DefaultOpacity := 1.0
                        DefaultDesiredSize := vector2{X := 512.0, Y := 64.0}  
                overlay_slot:
                    Widget := stack_box{
                        Orientation := orientation.Horizontal,
                        Slots := array{
                            stack_box_slot{
                                Widget := WhiteTimerText,
                                HorizontalAlignment := horizontal_alignment.Left,
                                VerticalAlignment := vertical_alignment.Center
                            },
                            stack_box_slot{
                                Widget := color_block:
                                    DefaultColor := NamedColors.CornflowerBlue
                                    DefaultOpacity := 0.0
                                    DefaultDesiredSize := vector2{X := 128.0, Y := 64.0}
                            },
                            stack_box_slot{
                                Widget := BlackTimerText,
                                HorizontalAlignment := horizontal_alignment.Right,
                                VerticalAlignment := vertical_alignment.Center
                            }
                        }
                    }
        
        #Adds a timer and winner/draw text underneath the score text above
        OverallUI := stack_box{
            Orientation := orientation.Vertical,
            Slots := array{
                stack_box_slot{
                    Widget := ScoreUI,
                    HorizontalAlignment := horizontal_alignment.Center,
                    VerticalAlignment := vertical_alignment.Center
                },
                stack_box_slot{
                    Widget := TurnStatusText,
                    HorizontalAlignment := horizontal_alignment.Center,
                    VerticalAlignment := vertical_alignment.Center
                },
                stack_box_slot{
                    Widget := ResultText,
                    HorizontalAlignment := horizontal_alignment.Center,
                    VerticalAlignment := vertical_alignment.Center
                }
            }
        }

        set MainWidget = canvas{
            Slots := array{
                canvas_slot{
                    Widget := OverallUI,
                    Anchors := anchors{Minimum := vector2{X := 0.5, Y := 0.1}, Maximum := vector2{X := 0.5, Y := 0.1}},
                    Alignment := vector2{X := 0.5, Y := 0.5}
                }
            }
        }

        Reset() 