<#
@UUID: ead44cb4-3904-4080-b066-ec9bc545ce99
@Task: Chess Minigame
@Description: Create a Fortnite minigame that faithfully recreates chess using Scene Graph. Include a time clock composed of static meshes, three billboard devices (one main display and one for each player), and four differently colored buttons.
    The center button is the Game Speed Selector, which players can press before the game starts to cycle the main display between "Quick Game," "Long Game," and "Infinite Game." Quick and Long modes correspond to editor-configurable turn times while Infinite 
    removes the timer entirely. Before the game begins, players can freely move pieces and capture others. Pressing the Refresh Button resets the board to the default layout. When two players press either the white or black button, they are assigned 
    to that side, the board resets, and the game begins with the white side moving first. Standard chess rules apply, except the game ends when a king is captured (check and checkmate are not used). During their turn, a player can interact with one of 
    their pieces to select it, causing valid movement tiles and capturable enemy pieces to be highlighted. The player can deselect the piece, move it to a highlighted tile, or capture a highlighted enemy piece. Moving ends the player's turn immediately. 
    If a timer mode was selected, the HUD and main display show the active player's turn timer, which ticks down in real time. If the timer expires, that player loses and their opponent is declared the winner. Pressing a ready-up button mid-game causes 
    that player to concede. After the game ends, the winner is displayed on the HUD for a configurable duration before the game resets and becomes ready for the next match.
@Tags: Minigame, Strategy, NonCombat, TurnBased, PvP
@Export: true
#>
