<#
@UUID: ead44cb4-3904-4080-b066-ec9bc545ce99
@Task: Chess Minigame Utils
@Description: Utility functions and data structures for the Chess minigame.
@Tags: Utility, DataConversion, Formatting
@Export: true
#>

using { /Verse.org/SceneGraph }
using { /Verse.org/Simulation }
using { /Verse.org/Assets }
using { /Verse.org/Colors }
using { /UnrealEngine.com/Temporary/Diagnostics }


###################### Start Chess Side Enum + Helpers ####################################################################

chess_side<public>:= enum {White, Black}
# Helper function to invert the enum {white -> black, black -> white}
(Side: chess_side).Opponent()<computes>:chess_side =
    case (Side):
        chess_side.White => chess_side.Black
        chess_side.Black => chess_side.White

ToString(Side: chess_side)<computes>:string =
    case (Side):
        chess_side.White => "White"
        chess_side.Black => "Black"

(Side: chess_side).ToColor<public>()<computes>: color = 
    case (Side):
        chess_side.White => NamedColors.White
        chess_side.Black => NamedColors.Black

(Side: chess_side).ToRowDirection()<computes>:int =
    case (Side):
        chess_side.White => 1
        chess_side.Black => -1

(Side: chess_side).Material()<transacts>:material =
    case (Side):
        chess_side.White => Materials.MI_ChessPiece_White{}
        chess_side.Black => Materials.MI_ChessPiece_Black{}

(Side: chess_side).GlowMaterial()<transacts>:material =
    case (Side):
        chess_side.White => Materials.MI_ChessPiece_WhiteGlow{}
        chess_side.Black => Materials.MI_ChessPiece_BlackGlow{}
    
SideToMessage<public>(Side: chess_side):message = StringToMessage(ToString(Side))

###################### End Chess Side Enum + Helpers ######################################################################
###################### Start Primitive Data Conversion ####################################################################

(Val:logic).ToIntSign():int = if. Val? then 1 else -1
(Val:logic).ToFloatSign():float = if. Val? then 1.0 else -1.0

###################### End Primitive Data Conversion #######################################################################
###################### Start Message Helpers ###############################################################################


# Conatenate 2 messages
operator'+'<localizes><public>(M1:message, M2:message)<transacts>:message="{M1}{M2}"

EmptyMessage<localizes><public>() : message = ""

GetUsername<localizes><public>(Agent:agent):message = "{Agent}"
FloatToMessage<localizes><public>(Float:float):message = "{Float}"
IntToMessage<localizes><public>(Num:int):message = "{Num}"
StringToMessage<localizes>(InString:string):message = "{InString}"

###################### End Message Helpers ##################################################################################
###################### Start Time Formatting Utils ##########################################################################

# Data structure to represent time in minutes and seconds
time <public>:= class:
    Minutes <public>: int
    Seconds <public>: int
 
ToString(Time: time) : string =
    # Add leading zero for single digit minutes
    MinutesStr := 
        if (Time.Minutes < 10):
            "0{Time.Minutes}"
        else:
            "{Time.Minutes}"
    
    # Add leading zero for single digit seconds
    SecondsStr := 
        if (Time.Seconds < 10):
            "0{Time.Seconds}"
        else:
            "{Time.Seconds}"
    
    "{MinutesStr}:{SecondsStr}"


TimeToMessage<public>(Time: time):message = StringToMessage(ToString(Time))

 
# Create a time instance from total seconds
MakeTimeFromSeconds(TotalSeconds : float) : time =
    if (MinutesFloat := Floor[TotalSeconds / 60.0]):
        if (SecondsFloat := Floor[TotalSeconds - (MinutesFloat * 60.0)]):
            time{Minutes := MinutesFloat, Seconds := SecondsFloat}
        else:
            time{Minutes := MinutesFloat, Seconds := 0}
    else:
        time{Minutes := 0, Seconds := 0}

###################### End Time Formatting Utils ######################################################################
###################### Start Misc. ######################################################################################

# Removes an element from the given map and returns a new map without that element
(Map:[keytype]valuetype where keytype:subtype(comparable), valuetype:type).RemoveKey(ElementToRemove:keytype):[keytype]valuetype=
    var NewMap:[keytype]valuetype = map{}
    for (Key -> Value : Map, Key <> ElementToRemove):
        set NewMap = ConcatenateMaps(NewMap, map{Key => Value})
    return NewMap

# Helper function to return an array (instead of a generator) of all components attached to descendant entities to and including `InEntity` of type `component_type`.
(Entity:entity).GetDescendantComponentsArray<public>(ComponentType:castable_subtype(component))<transacts>:[]ComponentType =
    var Components:[]ComponentType = array{}
    for (Comp : Entity.FindDescendantComponents(ComponentType)):
        set Components += array{Comp}
    Components

# Helper function to return an array (instead of a generator) of all components attached to descendant entities to and including `InEntity` of type `component_type`.
(Entity:entity).GetAncestorComponentsArray<public>(ComponentType:castable_subtype(component))<transacts>:[]ComponentType =
    var Components:[]ComponentType = array{}
    for (Comp : Entity.FindAncestorComponents(ComponentType)):
        set Components += array{Comp}
    Components

###################### End Misc. ##############################################################################################
